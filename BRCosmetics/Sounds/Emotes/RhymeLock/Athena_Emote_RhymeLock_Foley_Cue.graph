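# Foley sound cue for the RhymeLock emote, playing a jump sound followed by a landing sound.
# A mixer starts a random pick of three jump waves alongside a random pick of two land waves delayed by just under half a second.
# Each branch gets slight pitch and volume variation.
# The cue uses emote foley sound class, attenuation and concurrency settings, and lists a send to a music audio-analysis submix.
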SoundCue /BRCosmetics/Sounds/Emotes/RhymeLock/Athena_Emote_RhymeLock_Foley_Cue
root=Mixer_0
nodes [
 {id=Delay_1, children [Modulator_1], properties {delayMax=0.46, delayMin=0.45}}
 {id=Mixer_0, children [Modulator_0,Delay_1], properties {inputVolume [1,1]}}
 {id=Modulator_0, children [Random_2], properties {pitchMax=1, pitchMin=0.97, volumeMax=1.25, volumeMin=1}}
 {id=Modulator_1, children [Random_3], properties {pitchMax=1, pitchMin=0.97, volumeMax=1.25, volumeMin=1}}
 {id=Random_2, children [WavePlayer_0,WavePlayer_1,WavePlayer_2], properties {weights [1,1,1]}}
 {id=Random_3, children [WavePlayer_3, WavePlayer_11], properties {weights [1,1]}}
 {id=WavePlayer_0, properties {soundWaveAssetPtr=&/Rhyme_Lock_Jump_01}}
 {id=WavePlayer_1, properties {soundWaveAssetPtr=&/Rhyme_Lock_Jump_02}}
 {id=WavePlayer_11, properties {soundWaveAssetPtr=&/Rhyme_Lock_Land_02}}
 {id=WavePlayer_2, properties {soundWaveAssetPtr=&/Rhyme_Lock_Jump_03}}
 {id=WavePlayer_3, properties {soundWaveAssetPtr=&/Rhyme_Lock_Land_01}}
]
properties {
 attenuationSettings=&Fort_Emotes_Athena_Att
 audioPropertiesSheet=/BRCosmetics/Sounds/Settings/PropertySheets/Emotes/APS_Cosmetic_Emote_Foley
 maxDistance=1600
 soundClassObject=/Game/Sounds/FortSoundClassesAndModes/Fort_Foley_Emote
 volumeMultiplier=1
 concurrencyOverrides {resolutionRule=PreventNew}
 concurrencySet [/Game/Sounds/Concurrency_Rules/Athena_EmoteFoley_Conc]
 soundSubmixSends [{minSendLevel=0, soundSubmix=&Fortnite_AudioAnalysis_Music}]
}
