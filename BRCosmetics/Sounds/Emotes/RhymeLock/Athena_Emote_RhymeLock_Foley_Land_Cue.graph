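# Sound cue for the landing foley of the RhymeLock emote in Fortnite's Athena (Battle Royale) mode.
# It picks at random between two equally weighted land waves and applies slight pitch and volume variation through a modulator.
# It uses shared emote attenuation, the emote foley sound class and property sheet, and an emote foley concurrency set, and sends to a music audio-analysis submix.
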SoundCue /BRCosmetics/Sounds/Emotes/RhymeLock/Athena_Emote_RhymeLock_Foley_Land_Cue
root=Modulator_1
nodes [
 {id=Modulator_1, children [Random_3], properties {pitchMax=1, pitchMin=0.97, volumeMax=1.25, volumeMin=1}}
 {id=Random_3, children [WavePlayer_3, WavePlayer_11], properties {weights [1,1]}}
 {id=WavePlayer_11, properties {soundWaveAssetPtr=&/Rhyme_Lock_Land_02}}
 {id=WavePlayer_3, properties {soundWaveAssetPtr=&/Rhyme_Lock_Land_01}}
]
properties {
 attenuationSettings=&Fort_Emotes_Athena_Att
 audioPropertiesSheet=/BRCosmetics/Sounds/Settings/PropertySheets/Emotes/APS_Cosmetic_Emote_Foley
 maxDistance=1600
 soundClassObject=/Game/Sounds/FortSoundClassesAndModes/Fort_Foley_Emote
 volumeMultiplier=1
 concurrencyOverrides {resolutionRule=PreventNew}
 concurrencySet [/Game/Sounds/Concurrency_Rules/Athena_EmoteFoley_Conc]
 soundSubmixSends [{minSendLevel=0, soundSubmix=&Fortnite_AudioAnalysis_Music}]
}
