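# Looping music cue for the Rhyme Lock emote, using first-person emote music property sheet and concurrency settings.
# A licensed-content switcher selects either a silent loop or the music mix.
# The mix layers a loop under the 1P PSM sound class with an emote-class branch that crossfades idle and moving loops by EmoteMoveSpeed.
# Output also feeds a music audio-analysis submix.
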SoundCue /BRCosmetics/Sounds/Emotes/RhymeLock/Athena_Emote_RhymeLock_Music_Cue
root=FortSoundNodeLicensedContentSwitcher_0
nodes [
 {id=FortSoundNodeLicensedContentSwitcher_0, class=FortSoundNodeLicensedContentSwitcher, children [WavePlayer_1,Mixer_0,Mixer_0]}
 {id=Mixer_0, children [SoundClass_1,SoundClass_0], properties {inputVolume [1,0.75]}}
 {id=Mixer_1, children [WavePlayer_5], properties {inputVolume [0.75]}}
 {
  id=ParamCrossFade_0
  children [WavePlayer_4,Mixer_1]
  properties {
   paramName=EmoteMoveSpeed
   crossFadeInput [
    {fadeInDistanceEnd=0, fadeInDistanceStart=0, fadeOutDistanceEnd=1, fadeOutDistanceStart=0, volume=1}
    {fadeInDistanceEnd=1, fadeInDistanceStart=0, fadeOutDistanceEnd=100, fadeOutDistanceStart=100, volume=1}
   ]
  }
 }
 {id=SoundClass_0, children [ParamCrossFade_0], properties {soundClassOverride=&Fort_Music_Emote}}
 {id=SoundClass_1, children [WavePlayer_0], properties {soundClassOverride=/Game/Sounds/FortSoundClassesAndModes/Fort_Music_Emote_1P_PSM}}
 {id=WavePlayer_0, properties {bLooping=true, soundWaveAssetPtr=&/Emote_Ryhme_Lock_Loop}}
 {id=WavePlayer_1, properties {bLooping=true, soundWaveAssetPtr=/Game/Sounds/Fort_Test_Sounds/5_seconds_silence}}
 {id=WavePlayer_4, properties {bLooping=true, soundWaveAssetPtr=&/Rhyme_Lock_Idle}}
 {id=WavePlayer_5, properties {bLooping=true, soundWaveAssetPtr=&/Rhyme_Lock_Loop}}
]
properties {
 audioPropertiesSheet=/BRCosmetics/Sounds/Settings/PropertySheets/Emotes/APS_Cosmetic_Emote_Music_1P
 maxDistance=2097152
 soundClassObject=&Fort_Music_Emote
 volumeMultiplier=1
 concurrencyOverrides {maxCount=3, resolutionRule=StopOldest}
 concurrencySet [/Game/Sounds/Concurrency_Rules/Athena_EmoteMusic_1P_Conc]
 soundSubmixSends [{minSendLevel=0, soundSubmix=&Fortnite_AudioAnalysis_Music}]
}
